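# Sound cue for hovering over the Challenges tab in the Athena front-end UI, played in the UI sound class at reduced volume.
# A single main-menu tab hover wave feeds a Modulator that pins pitch and leaves volume modulation at its default.
# A concurrency override allows one instance at a time and stops the oldest, so rapid hovers cut off rather than stack.
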SoundCue /Game/Sounds/Fort_UI_Sounds/FrontEnd/Fort_Athena_ChallengesTab_Hover_Cue
root=Modulator_0
nodes [
 {id=Modulator_0, children [WavePlayer_1], properties {pitchMax=1, pitchMin=1}}
 {id=WavePlayer_1, properties {soundWaveAssetPtr=/Game/Sounds/Fort_UI_Sounds/FrontEnd/New_FrontEnd/Fortnite_UI_MainMenu_Tab_Hover_01}}
]
properties {
 bOverrideConcurrency=true
 maxDistance=2097152
 soundClassObject=&Fort_UI
 volumeMultiplier=0.5
 concurrencyOverrides {maxCount=1, resolutionRule=StopOldest}
}
