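# Sound cue for cancelling matchmaking in the front-end UI, judging by its name and location.
# A mixer layers a scroll click with a main button click and a hero-change whoosh, the latter two pitched down by fixed-value modulators.
# A concurrency override allows one instance, stopping the oldest, and the cue uses the application tray notification sound class.
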SoundCue /Game/Sounds/Fort_UI_Sounds/FrontEnd/Fort_Matchmaking_Cancel_Cue
root=Mixer_1
nodes [
 {id=Mixer_1, children [WavePlayer_2,Modulator_0,Modulator_1], properties {inputVolume [1.4,1,1.4]}}
 {id=Modulator_0, children [WavePlayer_0], properties {pitchMax=0.5, pitchMin=0.5}}
 {id=Modulator_1, children [WavePlayer_1], properties {pitchMax=0.8, pitchMin=0.8}}
 {id=WavePlayer_0, properties {soundWaveAssetPtr=/Game/Sounds/Fort_UI_Sounds/FrontEnd/New_FrontEnd/UI_FN_Button_Click_Main_01}}
 {id=WavePlayer_1, properties {soundWaveAssetPtr=/Game/Sounds/Fort_UI_Sounds/Gameplay/UI_ChangeHero_Whoosh_RightSide_01}}
 {id=WavePlayer_2, properties {soundWaveAssetPtr=/Game/Sounds/Fort_UI_Sounds/FrontEnd/New_FrontEnd/UI_FN_ScrollClick02}}
]
properties {
 bOverrideConcurrency=true
 maxDistance=2097152
 soundClassObject=/Game/Sounds/FortSoundClassesAndModes/Fort_Application_TrayNotification
 volumeMultiplier=1
 concurrencyOverrides {maxCount=1, resolutionRule=StopOldest}
}
