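# Sound cue for an ability click in the front-end UI.
# A single wave player feeds the output directly at reduced volume, with no randomization or modulation.
# It routes through the Fort_UI sound class and overrides concurrency so new plays are blocked, rather than cutting off playing ones, once the limit is reached.
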SoundCue /Game/Sounds/Fort_UI_Sounds/FrontEnd/Fort_UI_Ability_Click_Cue
root=WavePlayer_5
nodes [{id=WavePlayer_5, properties {soundWaveAssetPtr=&/fort_ui_new_chris_03}}]
properties {
 bOverrideConcurrency=true
 maxDistance=2097152
 soundClassObject=&Fort_UI
 volumeMultiplier=0.5
 concurrencyOverrides {resolutionRule=PreventNew}
}
