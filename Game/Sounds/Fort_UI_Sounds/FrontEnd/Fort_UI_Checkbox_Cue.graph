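# UI sound cue for a front-end checkbox, routed to the Fort_UI sound class at reduced volume.
# A Mixer layers the checkbox wave at full level with a quieter pulse wave.
# An Enveloper fades only the pulse layer to silence over roughly 1.3 seconds.
# The cue overrides attenuation and concurrency, with concurrency set to prevent new instances at the limit.
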SoundCue /Game/Sounds/Fort_UI_Sounds/FrontEnd/Fort_UI_Checkbox_Cue
root=Mixer_1
nodes [
 {
  id=Enveloper_0
  children [WavePlayer_14]
  properties {
   pitchCurve {editorCurveData {keys [{time=0}]}}
   volumeCurve {editorCurveData {keys [
      <at 0 1>
      {arriveTangent=-0.2718, leaveTangent=-0.2718, tangentMode=RCTM_User, time=0.6269, value=0.2833}
      {arriveTangent=-0.4257, interpMode=RCIM_Cubic, leaveTangent=-0.4257, tangentMode=RCTM_User, time=1.2738, value=-0.0196}
     ]}}
  }
 }
 {id=Mixer_1, children [WavePlayer_1,Enveloper_0], properties {inputVolume [1,0.3]}}
 {id=WavePlayer_1, properties {soundWaveAssetPtr=&/fort_ui_new_checkbox}}
 {id=WavePlayer_14, properties {soundWaveAssetPtr=&/fort_ui_pulse_01}}
]
properties {
 bOverrideAttenuation=true
 bOverrideConcurrency=true
 maxDistance=3000
 soundClassObject=&Fort_UI
 volumeMultiplier=0.4
 attenuationOverrides {falloffDistance=2600}
 concurrencyOverrides {resolutionRule=PreventNew}
}
