SoundCue /Game/Sounds/Fort_UI_Sounds/FrontEnd/Fort_UI_LTM_TileReveal
root=Mixer_0
nodes [
 {id=Mixer_0, children [Modulator_0,WavePlayer_0], properties {inputVolume [1,0.5]}}
 {id=Modulator_0, children [WavePlayer_11], properties {pitchMax=1.2, pitchMin=1.2}}
 {id=WavePlayer_0, properties {soundWaveAssetPtr=&/Fort_UI_CB_ItemView_Open}}
 {id=WavePlayer_11, properties {soundWaveAssetPtr=&/fort_rise_01}}
]
properties {
 bOverrideConcurrency=true
 maxDistance=2097152
 soundClassObject=&Fort_UI
 volumeMultiplier=0.5
 concurrencyOverrides {maxCount=1, resolutionRule=PreventNew}
}
