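# Front-end UI sound cue for a menu button click, routed to the Fort_UI sound class.
# A single wave player reuses the cancel button sample, Fort_Cancel_Button_01, through a Modulator with identical min and max values.
# The Modulator therefore applies a fixed raised pitch and lowered volume without randomization, with the cue's volume multiplier reducing it further.
# A shared concurrency asset, TwoMaximum_StopOldest_VolumeScale, governs overlapping clicks.
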SoundCue /Game/Sounds/Fort_UI_Sounds/FrontEnd/Fort_UI_MenuButtonClick_Cue
root=Modulator_0
nodes [
 {id=Modulator_0, children [WavePlayer_3], properties {pitchMax=1.2, pitchMin=1.2, volumeMax=0.4, volumeMin=0.4}}
 {id=WavePlayer_3, properties {soundWaveAssetPtr=&/Fort_Cancel_Button_01}}
]
properties {
 maxDistance=2097152
 soundClassObject=&Fort_UI
 volumeMultiplier=0.7
 concurrencyOverrides {resolutionRule=PreventNew}
 concurrencySet [/Game/Sounds/Concurrency_Rules/TwoMaximum_StopOldest_VolumeScale]
}
