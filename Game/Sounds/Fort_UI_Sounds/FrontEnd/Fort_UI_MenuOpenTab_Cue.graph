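# Sound cue for the front-end UI sound that, by its name, plays when a menu tab opens.
# A single wave player plays Fort_Menu_Misc_01 at reduced volume in the Fort_UI sound class, with no randomization or modulation.
# It overrides concurrency with a rule that prevents new instances from starting once the limit is reached.
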SoundCue /Game/Sounds/Fort_UI_Sounds/FrontEnd/Fort_UI_MenuOpenTab_Cue
root=WavePlayer_2
nodes [{id=WavePlayer_2, properties {soundWaveAssetPtr=&/Fort_Menu_Misc_01}}]
properties {
 bOverrideConcurrency=true
 maxDistance=2097152
 soundClassObject=&Fort_UI
 volumeMultiplier=0.5
 concurrencyOverrides {resolutionRule=PreventNew}
}
