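# Front-end UI sound cue that, going by its name, plays when the player selects a game.
# A mixer layers a main-menu start-game sound, a briefly delayed OK sound, a quiet click faded out by a volume envelope, and a sub-mix of pulse and locker-press sounds.
# It uses the application tray notification sound class and overrides concurrency so only one instance plays, stopping the oldest.
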SoundCue /Game/Sounds/Fort_UI_Sounds/FrontEnd/Fort_UI_SelectGame_Cue
root=Mixer_0
nodes [
 {id=Delay_2, children [WavePlayer_0], properties {delayMax=0.02, delayMin=0.02}}
 {
  id=Enveloper_6
  children [WavePlayer_27]
  properties {pitchCurve {editorCurveData {keys [{time=0}]}}, volumeCurve {editorCurveData {keys [<at 0 1>,<at 0.5 0>]}}}
 }
 {id=Mixer_0, children [WavePlayer_1,Delay_2,Enveloper_6,Mixer_3], properties {inputVolume [1.5,0.7,0.2,1]}}
 {id=Mixer_3, children [WavePlayer_2,WavePlayer_6], properties {inputVolume [1,1]}}
 {id=WavePlayer_0, properties {soundWaveAssetPtr=&/fort_ui_ok_02}}
 {id=WavePlayer_1, properties {soundWaveAssetPtr=/Game/Sounds/Fort_UI_Sounds/FrontEnd/New_FrontEnd/Fortnite_UI_MainMenu_StartGame_01}}
 {id=WavePlayer_2, properties {soundWaveAssetPtr=&/fort_ui_pulse_05}}
 {id=WavePlayer_27, properties {soundWaveAssetPtr=&/fort_new_frontend_ui_click_02}}
 {id=WavePlayer_6, properties {soundWaveAssetPtr=&/UI_FN_ItemandLocker_Pressed_01}}
]
properties {
 bOverrideConcurrency=true
 maxDistance=2097152
 soundClassObject=/Game/Sounds/FortSoundClassesAndModes/Fort_Application_TrayNotification
 concurrencyOverrides {maxCount=1, resolutionRule=StopOldest}
}
