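# Sound cue for a UI slottable slot change, stored with the front-end UI sounds.
# A single wave player feeds the slot-change wave straight to the output at reduced volume, with no randomization or modulation.
# It uses the in-game critical UI sound class and primes on load.
# A concurrency override allows one instance at a time, stopping the oldest when a new one plays.
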SoundCue /Game/Sounds/Fort_UI_Sounds/FrontEnd/Fort_UI_Slottable_Slot_Change_Cue
root=WavePlayer_1
nodes [{id=WavePlayer_1, properties {soundWaveAssetPtr=&/UI_Slottable_Slot_Change_01}}]
properties {
 bOverrideConcurrency=true
 bPrimeOnLoad=true
 maxDistance=2097152
 soundClassObject=&Fort_UI_InGame_Critical
 volumeMultiplier=0.5
 concurrencyOverrides {maxCount=1, resolutionRule=StopOldest}
 concurrencySet [null]
}
