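# Front-end UI sound cue that, going by its name, plays when a text message is received, in the Fort_UI sound class.
# A Mixer layers the UI_MessagePopup08 wave at full volume with a very quiet fort_UI_Info_Update wave.
# The quiet layer passes through a Modulator whose equal min and max pitch give a fixed upward shift rather than a random one.
# The cue overrides concurrency settings.
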SoundCue /Game/Sounds/Fort_UI_Sounds/FrontEnd/Fort_UI_TextMessage_RX_Cue
root=Mixer_8
nodes [
 {id=Mixer_8, children [WavePlayer_53, Modulator_8], properties {inputVolume [1,0.05]}}
 {id=Modulator_8, children [WavePlayer_52], properties {pitchMax=1.3, pitchMin=1.3}}
 {id=WavePlayer_52, properties {soundWaveAssetPtr=/Game/Sounds/Fort_UI_Sounds/Gameplay/fort_UI_Info_Update}}
 {id=WavePlayer_53, properties {soundWaveAssetPtr=/Game/Sounds/Fort_UI_Sounds/Gameplay/UI_MessagePopup08}}
]
properties {bOverrideConcurrency=true, maxDistance=2097152, soundClassObject=&Fort_UI}
