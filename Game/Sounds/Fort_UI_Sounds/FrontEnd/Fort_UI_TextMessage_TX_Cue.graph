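# Front-end UI sound cue for a text message event, likely the send (TX) side going by its name.
# A single wave player, the root node, plays the UI_MessagePopup11 wave with no randomization or modulation.
# The cue uses the Fort_UI sound class and sets its concurrency override.
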SoundCue /Game/Sounds/Fort_UI_Sounds/FrontEnd/Fort_UI_TextMessage_TX_Cue
root=WavePlayer_54
nodes [{id=WavePlayer_54, properties {soundWaveAssetPtr=/Game/Sounds/Fort_UI_Sounds/Gameplay/UI_MessagePopup11}}]
properties {bOverrideConcurrency=true, maxDistance=2097152, soundClassObject=&Fort_UI}
