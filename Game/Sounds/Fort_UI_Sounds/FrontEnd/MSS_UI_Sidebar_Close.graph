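# One-shot stereo MetaSound for the front-end UI sidebar closing, playing a right-side social sidebar close wave in the Fort_UI sound class.
# On play, the wave starts alongside an ADSR envelope whose release fires after a very short delay, well before the attack completes.
# The envelope multiplies both channels, and the wave's end finishes the source.
# Pitch and volume are raised, and concurrency allows one instance, stopping the oldest.
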MetaSoundSource /Game/Sounds/Fort_UI_Sounds/FrontEnd/MSS_UI_Sidebar_Close
inputs [
 Sound {type=WaveAsset}
 UE.Source.OnPlay {type=Trigger}
]
interfaces [UE.OutputFormat.Stereo:1.0, UE.Source.OneShot:1.0, UE.Source:1.1]
outputs [
 "UE.OutputFormat.Stereo.Audio:0" {type=Audio}
 "UE.OutputFormat.Stereo.Audio:1" {type=Audio}
 UE.Source.OneShot.OnFinished {type=Trigger}
]
pages [Default {
  input_defaults {Sound=/Game/Sounds/Fort_UI_Sounds/FrontEnd/Social/UI_Sidebar_Close_Right_05_less_highs, UE.Source.OnPlay=null}
  nodes [
   TriggerDelay_1 = UE.Trigger Delay(In=^UE.Source.OnPlay) {also {defaults {Delay Time=0}}}
   n1 = UE.Reroute(Value=@TriggerDelay_1.Out)
   TriggerDelay_0 = UE.Trigger Delay(In=@n1.Value) {also {defaults {Delay Time=0.03}}}
   ADSREnvelopeADSREnvelope_0 = ADSR Envelope.ADSR Envelope:Audio(Trigger Attack=@n1.Value, Trigger Release=@TriggerDelay_0.Out) {also {defaults {Attack Curve=2, Attack Time=0.5, Decay Curve=2, Decay Time=0, Release Curve=3.5, Release Time=0.8, Sustain Level=1}}}
   n2 = UE.Reroute(Value=@ADSREnvelopeADSREnvelope_0.Out Envelope)
   WavePlayer_0 = UE.Wave Player:Stereo(Wave Asset=^Sound, Play=@TriggerDelay_1.Out) {also {version [1,1]}}
   Multiply_0 = UE.Multiply:Audio(AdditionalOperands=@n2.Value, PrimaryOperand=@WavePlayer_0.Out Right)
   Multiply_1 = UE.Multiply:Audio(AdditionalOperands=@n2.Value, PrimaryOperand=@WavePlayer_0.Out Left)
  ]
  links [
   {dst="@UE.OutputFormat.Stereo.Audio:1", src=Multiply_0.Out}
   {dst="@UE.OutputFormat.Stereo.Audio:0", src=Multiply_1.Out}
   {dst="@UE.Source.OneShot.OnFinished", src=WavePlayer_0.On Finished}
  ]
 }]
properties {
 assetImportData=&self:AssetImportData
 bOverrideConcurrency=true
 numChannels=2
 outputFormat=Stereo
 pitch=2.178
 sampleRate=48000
 soundClassObject=&Fort_UI
 virtualizationMode=Disabled
 volume=2.4
 concurrencyOverrides {maxCount=1, resolutionRule=StopOldest}
}
