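# Sound cue for the matchmaking "match found" notification in the front-end UI.
# A single wave player plays the Battle Royale no-horn tray notification wave at reduced volume through the application tray notification sound class.
# A shared tray notification concurrency set is assigned, alongside override values that specify a single instance with stop-oldest resolution.
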SoundCue /Game/Sounds/Fort_UI_Sounds/FrontEnd/SC_UI_MatchMaking_MatchFound_Cue
root=WavePlayer_0
nodes [{id=WavePlayer_0, properties {soundWaveAssetPtr=/Game/Sounds/Fort_UI_Sounds/TrayNotifications/Fort_TrayNotification_MatchFound_BR_No_Horn_01}}]
properties {
 maxDistance=2097152
 soundClassObject=/Game/Sounds/FortSoundClassesAndModes/Fort_Application_TrayNotification
 volumeMultiplier=0.4
 concurrencyOverrides {maxCount=1, resolutionRule=StopOldest}
 concurrencySet [/Game/Sounds/Fort_UI_Sounds/TrayNotifications/CONC_TrayNotification]
}
