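# Sound cue for the default front-end matchmaking Play button press.
# A single wave player outputs the pressed sound directly, with no randomization or modulation.
# It routes through the application tray notification sound class and uses a dedicated Play-button concurrency asset.
# Inline concurrency values specify one stop-oldest instance, though no override flag appears to enable them over that asset.
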SoundCue /Game/Sounds/Fort_UI_Sounds/FrontEnd/SC_UI_MatchMaking_PlayButton_Pressed_Default
root=WavePlayer_1
nodes [{id=WavePlayer_1, properties {soundWaveAssetPtr=&/UI_MatchMaking_PlayButton_Pressed_01}}]
properties {
 maxDistance=2097152
 soundClassObject=/Game/Sounds/FortSoundClassesAndModes/Fort_Application_TrayNotification
 volumeMultiplier=1
 concurrencyOverrides {maxCount=1, resolutionRule=StopOldest}
 concurrencySet [&/CONC_MatchMaking_PlayButton]
}
