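# SC_UI_Party_Suggestion is a front-end UI Sound Cue in the Fort_UI sound class, named for party suggestions.
# It has no nodes or root node, so it produces no audio as defined.
# It keeps a reduced volume and overrides concurrency to cap simultaneous instances at a small count, stopping the oldest, with a short retrigger window.
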SoundCue /Game/Sounds/Fort_UI_Sounds/FrontEnd/SC_UI_Party_Suggestion
root=null
nodes []
properties {
 bOverrideConcurrency=true
 maxDistance=2097152
 soundClassObject=&Fort_UI
 volumeMultiplier=0.35
 concurrencyOverrides {maxCount=2, resolutionRule=StopOldest, retriggerTime=0.15}
}
