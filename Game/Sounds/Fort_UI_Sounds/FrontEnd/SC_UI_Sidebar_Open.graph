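# Sound cue that plays the sidebar-open UI sound in the front end.
# A single wave player feeds the right-side open sample straight to the output at reduced volume under the Fort UI sound class.
# A concurrency override allows one instance at a time, stopping the oldest, and virtualization is disabled.
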SoundCue /Game/Sounds/Fort_UI_Sounds/FrontEnd/SC_UI_Sidebar_Open
root=WavePlayer_3
nodes [{id=WavePlayer_3, properties {soundWaveAssetPtr=&/UI_Sidebar_Open_Right_01}}]
properties {
 bOverrideConcurrency=true
 maxDistance=2097152
 soundClassObject=&Fort_UI
 virtualizationMode=Disabled
 volumeMultiplier=0.5
 concurrencyOverrides {maxCount=1, resolutionRule=StopOldest}
}
