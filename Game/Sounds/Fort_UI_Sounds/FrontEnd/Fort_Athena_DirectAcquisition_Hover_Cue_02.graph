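# UI sound cue for the hover feedback in the Athena DirectAcquisition front-end screen.
# A Mixer layers a hover-click wave over a quieter button-hover wave, feeding a root Modulator that holds volume steady and leaves pitch at default variation.
# It uses the Fort_UI sound class with boosted volume, and a concurrency override allows one instance, blocking new plays rather than cutting off the current one.
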SoundCue /Game/Sounds/Fort_UI_Sounds/FrontEnd/Fort_Athena_DirectAcquisition_Hover_Cue_02
root=Modulator_2
nodes [
 {id=Mixer_1, children [WavePlayer_4,WavePlayer_3], properties {inputVolume [0.75,0.5]}}
 {id=Modulator_2, children [Mixer_1], properties {volumeMax=1, volumeMin=1}}
 {id=WavePlayer_3, properties {soundWaveAssetPtr=/Game/Sounds/Fort_UI_Sounds/FrontEnd/New_FrontEnd/UI_FN_Button_Hover}}
 {id=WavePlayer_4, properties {soundWaveAssetPtr=/Game/Sounds/Fort_UI_Sounds/FrontEnd/New_FrontEnd/UI_FN_Button_Hover_Click_02}}
]
properties {
 bOverrideConcurrency=true
 maxDistance=2097152
 soundClassObject=&Fort_UI
 volumeMultiplier=1.4
 concurrencyOverrides {maxCount=1, resolutionRule=PreventNew}
}
