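# Sound cue for the front-end party UI, by its name played when the pointer leaves a party element.
# A single wave player feeds the Party Add unhover wave into a modulator that raises pitch by a fixed amount and slightly randomizes volume.
# It uses the UI sound class at reduced volume with virtualization disabled.
# Its own concurrency override keeps one instance, stopping the oldest when retriggered.
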SoundCue /Game/Sounds/Fort_UI_Sounds/FrontEnd/Fort_UI_Party_UnHover_Cue
root=Modulator_0
nodes [
 {id=Modulator_0, children [WavePlayer_0], properties {pitchMax=1.25, pitchMin=1.25, volumeMax=1, volumeMin=0.9}}
 {id=WavePlayer_0, properties {soundWaveAssetPtr=&/Fort_UI_Party_Add_Unhover_01}}
]
properties {
 bOverrideConcurrency=true
 maxDistance=2097152
 soundClassObject=&Fort_UI
 virtualizationMode=Disabled
 volumeMultiplier=0.3
 concurrencyOverrides {maxCount=1, resolutionRule=StopOldest}
}
