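# UI sound cue for hovering over the skip button in cinematics.
# A mixer layers a hover-click wave over a quieter button hover wave, then feeds a modulator whose pitch and volume ranges are fixed, so playback has no variation.
# It uses the RenderedCinematics sound class with high priority and references a front-end menu concurrency rule that limits overlapping plays.
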SoundCue /Game/Sounds/Fort_UI_Sounds/FrontEnd/SC_UI_Cinematics_SkipButton_Hover
root=Modulator_0
nodes [
 {id=Mixer_0, children [WavePlayer_2,WavePlayer_1], properties {inputVolume [0.75,0.5]}}
 {id=Modulator_0, children [Mixer_0], properties {pitchMax=1, pitchMin=1, volumeMax=1, volumeMin=1}}
 {id=WavePlayer_1, properties {soundWaveAssetPtr=/Game/Sounds/Fort_UI_Sounds/FrontEnd/New_FrontEnd/UI_FN_Button_Hover}}
 {id=WavePlayer_2, properties {soundWaveAssetPtr=/Game/Sounds/Fort_UI_Sounds/FrontEnd/New_FrontEnd/UI_FN_Button_Hover_Click_02}}
]
properties {
 maxDistance=2097152
 priority=0.9
 soundClassObject=/Game/Sounds/FortSoundClassesAndModes/RenderedCinematics
 volumeMultiplier=1
 concurrencyOverrides {maxCount=1, resolutionRule=StopOldest}
 concurrencySet [/Game/Sounds/Concurrency_Rules/FrontEnd_UI_Menu_Max_03_Stop_Oldest]
}
