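# UI sound cue for pressing the skip button in cinematics.
# A mixer layers a main button click, a quieter deep-impact shop click and a whoosh that a modulator pitches down by a fixed amount.
# It plays under the RenderedCinematics sound class, with a concurrency override that rejects new instances at the limit instead of stopping playing ones.
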SoundCue /Game/Sounds/Fort_UI_Sounds/FrontEnd/SC_UI_Cinematics_SkipButton_Pressed
root=Mixer_0
nodes [
 {id=Mixer_0, children [WavePlayer_0,WavePlayer_1,Modulator_0], properties {inputVolume [0.8,0.5,0.8]}}
 {id=Modulator_0, children [WavePlayer_4], properties {pitchMax=0.8, pitchMin=0.8}}
 {id=WavePlayer_0, properties {soundWaveAssetPtr=/Game/Sounds/Fort_UI_Sounds/FrontEnd/New_FrontEnd/UI_FN_Button_Click_Main_01}}
 {id=WavePlayer_1, properties {soundWaveAssetPtr=/Game/Sounds/Fort_UI_Sounds/FrontEnd/ItemShop/wav/Fort_UI_Shop_Click_DeepImpact_01}}
 {id=WavePlayer_4, properties {soundWaveAssetPtr=&/UI_Cinematics_Skip_Button_Pressed_Whoosh}}
]
properties {
 bOverrideConcurrency=true
 maxDistance=2097152
 soundClassObject=/Game/Sounds/FortSoundClassesAndModes/RenderedCinematics
 volumeMultiplier=1
 concurrencyOverrides {resolutionRule=PreventNew}
}
