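# Sound cue for the matchmaking timer tick, filed under front-end UI sounds.
# A single wave player feeds the root, reusing the Creative floor siren's Timer_Tick_01 wave at reduced volume.
# It routes through the non-critical in-game UI sound class.
# A concurrency override allows only one instance and blocks new plays while one is active, so ticks never overlap.
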SoundCue /Game/Sounds/Fort_UI_Sounds/FrontEnd/SC_UI_MatchMaking_TimeTick_Cue
root=WavePlayer_0
nodes [{id=WavePlayer_0, properties {soundWaveAssetPtr=/Game/Sounds/Creative/Toys/Floor_Siren/Timer_Tick_01}}]
properties {
 bOverrideConcurrency=true
 maxDistance=2097152
 soundClassObject=&Fort_UI_InGame_NonCritical
 volumeMultiplier=0.5
 concurrencyOverrides {maxCount=1, resolutionRule=PreventNew}
}
