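# Sound cue for the front-end UI sound that, by its name, plays when a player un-readies from matchmaking.
# A single Wave Player feeds the ReadyUp_Disable wave straight to the output at reduced volume, with no randomisation or modulation.
# It routes through the application tray notification sound class, disables virtualization, and shares the pressed-button concurrency set to limit overlapping plays.
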SoundCue /Game/Sounds/Fort_UI_Sounds/FrontEnd/SC_UI_MatchMaking_UnReady_Cue
root=WavePlayer_0
nodes [{id=WavePlayer_0, properties {soundWaveAssetPtr=&/UI_MatchMaking_ReadyUp_Disable_01}}]
properties {
 maxDistance=2097152
 soundClassObject=/Game/Sounds/FortSoundClassesAndModes/Fort_Application_TrayNotification
 virtualizationMode=Disabled
 volumeMultiplier=0.5
 concurrencyOverrides {maxCount=1, resolutionRule=StopOldest}
 concurrencySet [/Game/Sounds/Fort_UI_Sounds/FrontEnd/DefaultSounds/PressedButtons/UI_PressedButton_Concurrency]
}
