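# Sound cue for a front-end UI dismiss sound, the third numbered variant by name.
# A single wave player feeds the root directly and plays the fort_ui_dismiss_03 wave, with no randomization or modulation nodes.
# The cue routes to the Fort_UI sound class, plays at reduced volume and sets a very large max distance.
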SoundCue /Game/Sounds/Fort_UI_Sounds/FrontEnd/fort_ui_dismiss_03_Cue
root=WavePlayer_0
nodes [{id=WavePlayer_0, properties {soundWaveAssetPtr=&/fort_ui_dismiss_03}}]
properties {maxDistance=2097152, soundClassObject=&Fort_UI, volumeMultiplier=0.5}
